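# CRU_PropAim is a Control Rig utility asset, named for prop aiming, that holds only a skeleton setup with no aiming logic.
# Its hierarchy has a root and three Prop connectors for Parent, Spaces and Control Vis Channel Host, the attachment points of a modular rig module.
# The graph contains only an unconnected Begin Execution event and the function library is empty, so the file shows no controls or solve.
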ControlRigBlueprint /EpicControlRig/UtilityRigs/CRU_PropAim
graphs [
 RigVMFunctionLibrary {links [], nodes []}
 RigVMModel {nodes [RigUnit_BeginExecution = &RigUnit_BeginExecution()]}
]
hierarchy [
 root {loc=<0 0 0>, rot=<0 0 0 1>}
 Prop/Parent {type=Connector, loc=<0 0 0>, rot=<0 0 0 1>}
 Prop/Spaces {type=Connector, loc=<0 0 0>, rot=<0 0 0 1>}
 Prop/Control Vis Channel Host {type=Connector, loc=<0 0 0>, rot=<0 0 0 1>}
]
