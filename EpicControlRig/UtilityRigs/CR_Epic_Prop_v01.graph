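# Utility Control Rig for props that builds its controls procedurally from the imported skeleton.
# Construction spawns Global, optional Local and per-bone controls, skipping "dyn_" bones, with shape and color chosen by name and scale reduced for stacked controls.
# Forward solve drives bones from controls and toggles Hand and Attach control visibility through bool channels on Global.
# Backwards solve sets controls from bone transforms.
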ControlRigBlueprint /EpicControlRig/UtilityRigs/CR_Epic_Prop_v01
graphs [
 RigVMFunctionLibrary {links [], nodes []}
 RigVMFunction_Sequence_1_ContainedGraph {nodes [
   n1 = RigVMFunctionEntryNode()
   RigVMFunction_Sequence = &RigVMFunction_Sequence(ExecuteContext=@n1.ExecuteContext)
   n2 = &RigVMFunction_Sequence(ExecuteContext=@RigVMFunction_Sequence.B)
   n3 = &RigVMFunction_Sequence(ExecuteContext=@n2.B)
   RigVMFunctionReturnNode(A=@RigVMFunction_Sequence.A, B=@n2.A, C=@n3.A, D=@n3.B)
  ]}
 RigVMModel {nodes [
   RigUnit_BeginExecution = &RigUnit_BeginExecution()
   RigVMFunction_Sequence = RigVMAggregateNode(ExecuteContext=@RigUnit_BeginExecution.ExecutePin)
   For_Each_4 = RigVMDispatchNode(Array=RigVMVariableNode() {variable=Attach Controls}.Value, ExecuteContext=@RigVMFunction_Sequence.D) {template="DISPATCH_RigVMDispatch_ArrayIterator(in Array,out Element,out Index,out Count,out Ratio)"}
   > For_Each_5 = RigVMDispatchNode(Array=RigVMVariableNode() {variable=Attach Vis Channels}.Value) {template="DISPATCH_RigVMDispatch_ArrayIterator(in Array,out Element,out Index,out Count,out Ratio)"}
   n1 = &RigVMFunction_ControlFlowBranch(
    Condition=RigVMDispatchNode(
     A=/Script/RigVM.RigVMFunction_NameConcat(B=" Vis", A=@For_Each_4.Element.Name) {is=struct RigVMUnitNode, template="Concat::Execute(in A,in B,out Result)"}.Result
     B=@For_Each_5.Element.Name
    ) {template="DISPATCH_RigVMDispatch_CoreEquals(in A,in B,out Result)"}.Result
    ExecuteContext=@For_Each_5.ExecuteContext
   )
   /Script/ControlRig.RigUnit_SetControlVisibility(
    bVisible=/Script/ControlRig.RigUnit_GetBoolAnimationChannelFromItem(bInitial=false, Item=@For_Each_5.Element) {is=struct RigVMUnitNode, template="GetAnimationChannelFromItem::Execute(out Value,in Item,in bInitial)"}.Value
    ExecutePin=@n1.True
    Item=@For_Each_4.Element
   ) {is=struct RigVMUnitNode, template="SetControlVisibility::Execute(in Item,in Pattern,in bVisible)"}
   For_Each = RigVMDispatchNode(Array=RigVMVariableNode() {variable=Driven Children Bones}.Value, ExecuteContext=@RigVMFunction_Sequence.B) {template="DISPATCH_RigVMDispatch_ArrayIterator(in Array,out Element,out Index,out Count,out Ratio)"}
   n2 = RigVMVariableNode() {variable=Children Controls}
   > &RigUnit_SetTransform(
    Weight=1
    bInitial=False
    bPropagateToChildren=true
    Value=&RigUnit_GetTransform(
     bInitial=False
     Item=RigVMDispatchNode(Index=@For_Each.Index, Array=@n2.Value) {template="DISPATCH_RigVMDispatch_ArrayGetAtIndex(in Array,in Index,out Element)"}.Element
    ).Transform
    Item=@For_Each.Element
   )
   RigVMFunction_ControlFlowBranch = &RigVMFunction_ControlFlowBranch(
    Condition=/Script/ControlRig.RigUnit_ItemExists(Item=RigVMVariableNode() {variable=Offset}.Value) {is=struct RigVMUnitNode}.Exists
    ExecuteContext=@RigVMFunction_Sequence.A
   )
   &RigUnit_SetTransform(
    Weight=1
    bInitial=False
    Value=&RigUnit_GetTransform(bInitial=False, Item=RigVMVariableNode() {variable=Global Control}.Value).Transform
    Item=RigVMVariableNode() {variable=Root}.Value
    ExecutePin=@RigVMFunction_ControlFlowBranch.False
   )
   &RigUnit_SetTransform(
    Weight=1
    bInitial=False
    Value=&RigUnit_GetTransform(bInitial=False, Item=RigVMVariableNode() {variable=Global Control}.Value).Transform
    Item=RigVMVariableNode() {variable=Root}.Value
    ExecutePin=@RigVMFunction_ControlFlowBranch.True
   )
   > &RigUnit_SetTransform(
    Weight=1
    bInitial=False
    Value=&RigUnit_GetTransform(bInitial=False, Item=RigVMVariableNode() {variable=Local Control}.Value).Transform
    Item=RigVMVariableNode() {variable=Offset}.Value
   )
   For_Each_1 = RigVMDispatchNode(Array=RigVMVariableNode() {variable=Hand Controls}.Value, ExecuteContext=@RigVMFunction_Sequence.C) {template="DISPATCH_RigVMDispatch_ArrayIterator(in Array,out Element,out Index,out Count,out Ratio)"}
   > For_Each_2 = RigVMDispatchNode(Array=RigVMVariableNode() {variable=Hand Vis Channels}.Value) {template="DISPATCH_RigVMDispatch_ArrayIterator(in Array,out Element,out Index,out Count,out Ratio)"}
   Concat = &RigVMFunction_NameConcat(B=" Vis", A=@For_Each_1.Element.Name) {template="Concat::Execute(in A,in B,out Result)"}
   n3 = &RigVMFunction_ControlFlowBranch(
    Condition=RigVMDispatchNode(A=@Concat.Result, B=@For_Each_2.Element.Name) {template="DISPATCH_RigVMDispatch_CoreEquals(in A,in B,out Result)"}.Result
    ExecuteContext=@For_Each_2.ExecuteContext
   )
   GetAnimationChannelFromItem = &RigUnit_GetBoolAnimationChannelFromItem(bInitial=false, Item=@For_Each_2.Element) {template="GetAnimationChannelFromItem::Execute(out Value,in Item,in bInitial)"}
   SetControlVisibility = &RigUnit_SetControlVisibility(bVisible=@GetAnimationChannelFromItem.Value, ExecutePin=@n3.True, Item=@For_Each_1.Element) {template="SetControlVisibility::Execute(in Item,in Pattern,in bVisible)"}
  ]}
 RigVMModel Backwards Solve Graph {nodes [
   n1 = /Script/ControlRig.RigUnit_InverseExecution() {is=struct RigVMUnitNode}
   RigVMFunction_Sequence = &RigVMFunction_Sequence(ExecuteContext=@n1.ExecutePin)
   n2 = &RigUnit_SetTransform(
    Weight=1
    bInitial=False
    Value=&RigUnit_GetTransform(bInitial=False, Item=RigVMVariableNode() {variable=Root}.Value).Transform
    Item=RigVMVariableNode() {variable=Global Control}.Value
    ExecutePin=@RigVMFunction_Sequence.A
   )
   RigVMFunction_ControlFlowBranch = &RigVMFunction_ControlFlowBranch(Condition=&RigUnit_ItemExists(Item=RigVMVariableNode() {variable=Offset}.Value).Exists, ExecuteContext=@n2.ExecutePin)
   &RigUnit_SetTransform(
    Weight=1
    bInitial=False
    Value=&RigUnit_GetTransform(bInitial=False, Item=RigVMVariableNode() {variable=Offset}.Value).Transform
    ExecutePin=@RigVMFunction_ControlFlowBranch.True
    Item=RigVMVariableNode() {variable=Local Control}.Value
   )
   For_Each = RigVMDispatchNode(Array=RigVMVariableNode() {variable=Children Controls}.Value, ExecuteContext=@RigVMFunction_Sequence.B) {template="DISPATCH_RigVMDispatch_ArrayIterator(in Array,out Element,out Index,out Count,out Ratio)"}
   n3 = RigVMVariableNode() {variable=Driven Children Bones}
   > &RigUnit_SetTransform(
    Weight=1
    bInitial=False
    bPropagateToChildren=false
    Value=&RigUnit_GetTransform(
     bInitial=False
     Item=RigVMDispatchNode(Index=@For_Each.Index, Array=@n3.Value) {template="DISPATCH_RigVMDispatch_ArrayGetAtIndex(in Array,in Index,out Element)"}.Element
    ).Transform
    Item=@For_Each.Element
   )
  ]}
 RigVMModel Construction Event Graph {nodes [
   n1 = &RigUnit_PrepareForExecution()
   &RigUnit_HierarchyImportFromSkeleton(NameSpace=None, bIncludeCurves=False, bIncludeMeshSockets=False, bIncludeVirtualBones=True, ExecutePin=@n1.ExecutePin)
   > RigVMVariableNode(Value=(Type=Bone,Name="root")) {variable=Root}
   > n2 = RigVMVariableNode(Value=(Type=Bone,Name="offset")) {variable=Offset}
   RigVMFunction_Sequence = &RigVMFunction_Sequence(ExecuteContext=@n2.ExecuteContext)
   SpawnControl = /Script/ControlRig.RigUnit_HierarchyAddControlTransform(
    InitialValue=(Rotation=(X=0,Y=0,Z=0,W=1),Translation=(X=0,Y=0,Z=0),Scale3D=(X=1,Y=1,Z=1))
    Name=Global
    OffsetSpace=GlobalSpace
    Parent=(Type=Bone,Name="None")
    Settings="(InitialSpace=LocalSpace,bUsePreferredRotationOrder=False,PreferredRotationOrder=YZX,Limits=(LimitTranslationX=(bMinimum=False,bMaximum=False),LimitTranslationY=(bMinimum=False,bMaximum=False),LimitTranslationZ=(bMinimum=False,bMaximum=False),LimitPitch=(b ...(truncated)"
    OffsetTransform=&RigUnit_GetTransform(bInitial=true, Item=RigVMVariableNode() {variable=Root}.Value).Transform
    ExecutePin=@RigVMFunction_Sequence.A
   ) {
    is=struct RigVMUnitNode
    template="SpawnControl::Execute(in InitialValue,in Settings,in OffsetTransform,in OffsetSpace,in Parent,in Name,out Item)"
   }
   > n3 = RigVMVariableNode(Value=@SpawnControl.Item) {variable=Global Control}
   RigVMFunction_ControlFlowBranch = &RigVMFunction_ControlFlowBranch(Condition=&RigUnit_ItemExists(Item=RigVMVariableNode() {variable=Offset}.Value).Exists, ExecuteContext=@n3.ExecuteContext)
   n4 = &RigUnit_HierarchyAddControlTransform(
    InitialValue=(Rotation=(X=0,Y=0,Z=0,W=1),Translation=(X=0,Y=0,Z=0),Scale3D=(X=1,Y=1,Z=1))
    Name=Local
    OffsetSpace=GlobalSpace
    Settings="(InitialSpace=LocalSpace,bUsePreferredRotationOrder=False,PreferredRotationOrder=YZX,Limits=(LimitTranslationX=(bMinimum=False,bMaximum=False),LimitTranslationY=(bMinimum=False,bMaximum=False),LimitTranslationZ=(bMinimum=False,bMaximum=False),LimitPitch=(b ...(truncated)"
    OffsetTransform=&RigUnit_GetTransform(bInitial=true, Item=RigVMVariableNode() {variable=Offset}.Value).Transform
    ExecutePin=@RigVMFunction_ControlFlowBranch.True
    Parent=RigVMVariableNode() {variable=Global Control}.Value
   ) {template="SpawnControl::Execute(in InitialValue,in Settings,in OffsetTransform,in OffsetSpace,in Parent,in Name,out Item)"}
   > RigVMVariableNode(Value=@n4.Item) {variable=Local Control}
   RigVMDispatchNode(Array=RigVMVariableNode() {variable=Children Controls}.Value, ExecuteContext=@RigVMFunction_Sequence.B) {template="DISPATCH_RigVMDispatch_ArrayReset(io Array)"}
   > RigVMDispatchNode(Array=RigVMVariableNode() {variable=Hand Vis Channels}.Value) {template="DISPATCH_RigVMDispatch_ArrayReset(io Array)"}
   > RigVMDispatchNode(Array=RigVMVariableNode() {variable=Hand Controls}.Value) {template="DISPATCH_RigVMDispatch_ArrayReset(io Array)"}
   > RigVMDispatchNode(Array=RigVMVariableNode() {variable=Attach Vis Channels}.Value) {template="DISPATCH_RigVMDispatch_ArrayReset(io Array)"}
   > RigVMDispatchNode(Array=RigVMVariableNode() {variable=Driven Children Bones}.Value) {template="DISPATCH_RigVMDispatch_ArrayReset(io Array)"}
   n5 = &RigUnit_ItemExists(Item=RigVMVariableNode() {variable=Offset}.Value)
   n6 = RigVMVariableNode() {variable=Root}
   n7 = RigVMVariableNode() {variable=Offset}
   > RigVMVariableNode(Value=/Script/ControlRig.RigUnit_CollectionChildrenArray(
     TypeToSearch=Bone
     bDefaultChildren=True
     bIncludeParent=False
     bRecursive=true
     Parent=RigVMDispatchNode(Condition=@n5.Exists, False=@n6.Value, True=@n7.Value) {template="DISPATCH_RigVMDispatch_If(in Condition,in True,in False,out Result)"}.Result
    ) {is=struct RigVMUnitNode}.Items) {variable=All Children Bones}
   > For_Each = RigVMDispatchNode(Array=RigVMVariableNode() {variable=All Children Bones}.Value) {template="DISPATCH_RigVMDispatch_ArrayIterator(in Array,out Element,out Index,out Count,out Ratio)"}
   To_String = RigVMDispatchNode(Value=@For_Each.Element.Name) {template="DISPATCH_RigDispatch_ToString(in Value,out Result)"}
   RigVMFunction_StringToLowercase = /Script/RigVM.RigVMFunction_StringToLowercase(Value=@To_String.Result) {is=struct RigVMUnitNode}
   n8 = &RigVMFunction_ControlFlowBranch(
    Condition=/Script/RigVM.RigVMFunction_StringStartsWith(Start=dyn_, Name=@RigVMFunction_StringToLowercase.Result) {is=struct RigVMUnitNode, template="StartsWith::Execute(in Name,in Start,out Result)"}.Result
    ExecuteContext=@For_Each.ExecuteContext
   )
   n9 = &RigVMFunction_Sequence(ExecuteContext=@n8.False)
   RigVMDispatchNode(ExecuteContext=@n9.B, Array=RigVMVariableNode() {variable=Driven Children Bones}.Value, Element=@For_Each.Element) {template="DISPATCH_RigVMDispatch_ArrayAdd(io Array,in Element,out Index)"}
   > Conform Name v01 = RigVMFunctionReferenceNode(Remove Side Suffix=false, Name=@For_Each.Element.Name)
   > Conform Name v01_1 = RigVMFunctionReferenceNode(
    Remove Side Suffix=false
    Name=/Script/ControlRig.RigUnit_HierarchyGetParent(bDefaultParent=True, Child=@For_Each.Element) {is=struct RigVMUnitNode}.Parent.Name
   )
   n10 = &RigUnit_ItemExists(Item=RigVMVariableNode() {variable=Offset}.Value)
   n11 = RigVMVariableNode() {variable=Local Control}
   n12 = RigVMVariableNode() {variable=Global Control}
   EndsWith = /Script/RigVM.RigVMFunction_EndsWith(Ending=" L", Name=@"Conform Name v01".Conformed Name) {is=struct RigVMUnitNode, template="EndsWith::Execute(in Name,in Ending,out Result)"}
   n13 = RigVMDispatchNode(
    False=(R=1,G=1,B=0,A=1)
    True=(R=1,G=0,B=0,A=1)
    Condition=&RigVMFunction_EndsWith(Ending=" R", Name=@"Conform Name v01".Conformed Name) {template="EndsWith::Execute(in Name,in Ending,out Result)"}.Result
   ) {template="DISPATCH_RigVMDispatch_If(in Condition,in True,in False,out Result)"}
   n14 = /Script/RigVM.RigVMFunction_StartsWith(Start=Attach, Name=@"Conform Name v01".Conformed Name) {is=struct RigVMUnitNode, template="StartsWith::Execute(in Name,in Start,out Result)"}
   RigVMFunction_MathBoolOr = &RigVMFunction_MathBoolOr(
    A=@n14.Result
    B=&RigVMFunction_StartsWith(Start="Hand Attach ", Name=@"Conform Name v01".Conformed Name) {template="StartsWith::Execute(in Name,in Start,out Result)"}.Result
   )
   StartsWith = &RigVMFunction_StartsWith(Start="Hand ", Name=@"Conform Name v01".Conformed Name) {template="StartsWith::Execute(in Name,in Start,out Result)"}
   n15 = RigVMDispatchNode(False=Sphere_Thick, True=Diamond_Thick, Condition=@StartsWith.Result) {template="DISPATCH_RigVMDispatch_If(in Condition,in True,in False,out Result)"}
   n16 = &RigVMFunction_StartsWith(Start=Handle, Name=@"Conform Name v01".Conformed Name) {template="StartsWith::Execute(in Name,in Start,out Result)"}
   Power = /Script/RigVM.RigVMFunction_MathDoublePow(A=RigVMVariableNode() {variable=Scale Factor}.Value, B=RigVMVariableNode() {variable=Control Stack}.Value) {is=struct RigVMUnitNode, template="Power::Execute(in A,in B,out Result)"}
   n17 = RigVMDispatchNode(True=2.5, False=@Power.Result, Condition=@n16.Result) {template="DISPATCH_RigVMDispatch_If(in Condition,in True,in False,out Result)"}
   > n18 = &RigUnit_HierarchyAddControlTransform(
    InitialValue=(Rotation=(X=0,Y=0,Z=0,W=1),Translation=(X=0,Y=0,Z=0),Scale3D=(X=1,Y=1,Z=1))
    OffsetSpace=GlobalSpace
    Settings="(InitialSpace=LocalSpace,bUsePreferredRotationOrder=False,PreferredRotationOrder=YZX,Limits=(LimitTranslationX=(bMinimum=False,bMaximum=False),LimitTranslationY=(bMinimum=False,bMaximum=False),LimitTranslationZ=(bMinimum=False,bMaximum=False),LimitPitch=(b ...(truncated)"
    Name=@"Conform Name v01".Conformed Name
    OffsetTransform=&RigUnit_GetTransform(bInitial=False, Item=@For_Each.Element).Transform
    Parent=RigVMDispatchNode(
     True=(Type=Control,Name="None")
     True.Name=@"Conform Name v01_1".Conformed Name
     False=RigVMDispatchNode(Condition=@n10.Exists, True=@n11.Value, False=@n12.Value) {template="DISPATCH_RigVMDispatch_If(in Condition,in True,in False,out Result)"}.Result
     Condition=&RigUnit_ItemExists(Item=(Type=Control,Name="None"), Item.Name=@"Conform Name v01_1".Conformed Name).Exists
    ) {template="DISPATCH_RigVMDispatch_If(in Condition,in True,in False,out Result)"}.Result
    Settings.Shape.Color=RigVMDispatchNode(
     True=(R=0,G=1,B=1,A=1)
     False=RigVMDispatchNode(True=(R=0,G=0,B=1,A=1), Condition=@EndsWith.Result, False=@n13.Result) {template="DISPATCH_RigVMDispatch_If(in Condition,in True,in False,out Result)"}.Result
     Condition=@RigVMFunction_MathBoolOr.Result
    ) {template="DISPATCH_RigVMDispatch_If(in Condition,in True,in False,out Result)"}.Result
    Settings.Shape.Name=RigVMDispatchNode(
     True=RoundedSquare_Thick
     False=RigVMDispatchNode(True=Box_thick, False=@n15.Result, Condition=@RigVMFunction_MathBoolOr.Result) {template="DISPATCH_RigVMDispatch_If(in Condition,in True,in False,out Result)"}.Result
     Condition=@n16.Result
    ) {template="DISPATCH_RigVMDispatch_If(in Condition,in True,in False,out Result)"}.Result
    Settings.Shape.Transform.Scale3D.X=@n17.Result
    Settings.Shape.Transform.Scale3D.Y=@n17.Result
    Settings.Shape.Transform.Scale3D.Z=@n17.Result
   ) {template="SpawnControl::Execute(in InitialValue,in Settings,in OffsetTransform,in OffsetSpace,in Parent,in Name,out Item)"}
   > Add = RigVMDispatchNode(Element=@n18.Item, Array=RigVMVariableNode() {variable=Children Controls}.Value) {template="DISPATCH_RigVMDispatch_ArrayAdd(io Array,in Element,out Index)"}
   n19 = &RigVMFunction_Sequence(ExecuteContext=@Add.ExecuteContext)
   n20 = &RigVMFunction_ControlFlowBranch(ExecuteContext=@n19.A, Condition=@StartsWith.Result)
   Concat = &RigVMFunction_NameConcat(B=" Vis", A=@"Conform Name v01".Conformed Name) {template="Concat::Execute(in A,in B,out Result)"}
   SpawnAnimationChannel = /Script/ControlRig.RigUnit_HierarchyAddAnimationChannelBool(
    InitialValue=false
    LimitsEnabled=()
    MaximumValue=false
    MinimumValue=false
    ExecutePin=@n20.True
    Parent=RigVMVariableNode() {variable=Global Control}.Value
    Name=@Concat.Result
   ) {
    is=struct RigVMUnitNode
    template="SpawnAnimationChannel::Execute(in InitialValue,in MinimumValue,in MaximumValue,in LimitsEnabled,in Parent,in Name,out Item)"
   }
   > RigVMDispatchNode(Element=@SpawnAnimationChannel.Item, Array=RigVMVariableNode() {variable=Hand Vis Channels}.Value) {template="DISPATCH_RigVMDispatch_ArrayAdd(io Array,in Element,out Index)"}
   > For_Each_3 = RigVMDispatchNode(Array=RigVMVariableNode() {variable=Children Controls}.Value) {template="DISPATCH_RigVMDispatch_ArrayIterator(in Array,out Element,out Index,out Count,out Ratio)"}
   n21 = &RigVMFunction_ControlFlowBranch(
    ExecuteContext=@For_Each_3.ExecuteContext
    Condition=&RigVMFunction_StartsWith(Start="Hand ", Name=@For_Each_3.Element.Name) {template="StartsWith::Execute(in Name,in Start,out Result)"}.Result
   )
   RigVMDispatchNode(Element=@For_Each_3.Element, ExecuteContext=@n21.True, Array=RigVMVariableNode() {variable=Hand Controls}.Value) {template="DISPATCH_RigVMDispatch_ArrayAdd(io Array,in Element,out Index)"}
   n22 = &RigVMFunction_ControlFlowBranch(ExecuteContext=@n19.B, Condition=@n14.Result)
   n23 = &RigUnit_HierarchyAddAnimationChannelBool(
    InitialValue=false
    LimitsEnabled=()
    MaximumValue=false
    MinimumValue=false
    ExecutePin=@n22.True
    Parent=RigVMVariableNode() {variable=Global Control}.Value
    Name=@Concat.Result
   ) {template="SpawnAnimationChannel::Execute(in InitialValue,in MinimumValue,in MaximumValue,in LimitsEnabled,in Parent,in Name,out Item)"}
   > RigVMDispatchNode(Element=@n23.Item, Array=RigVMVariableNode() {variable=Attach Vis Channels}.Value) {template="DISPATCH_RigVMDispatch_ArrayAdd(io Array,in Element,out Index)"}
   > For_Each_1 = RigVMDispatchNode(Array=RigVMVariableNode() {variable=Children Controls}.Value) {template="DISPATCH_RigVMDispatch_ArrayIterator(in Array,out Element,out Index,out Count,out Ratio)"}
   n24 = &RigVMFunction_ControlFlowBranch(
    ExecuteContext=@For_Each_1.ExecuteContext
    Condition=&RigVMFunction_StartsWith(Start="Attach ", Name=@For_Each_1.Element.Name) {template="StartsWith::Execute(in Name,in Start,out Result)"}.Result
   )
   RigVMDispatchNode(Element=@For_Each_1.Element, ExecuteContext=@n24.True, Array=RigVMVariableNode() {variable=Attach Controls}.Value) {template="DISPATCH_RigVMDispatch_ArrayAdd(io Array,in Element,out Index)"}
   RigVMVariableNode(ExecuteContext=@n9.A) {variable=Control Stack}
   > For_Each_2 = RigVMDispatchNode(Array=RigVMVariableNode() {variable=Children Controls}.Value) {template="DISPATCH_RigVMDispatch_ArrayIterator(in Array,out Element,out Index,out Count,out Ratio)"}
   Contains = /Script/RigVM.RigVMFunction_Contains(Search="Hand ", Name=@For_Each_2.Element.Name) {is=struct RigVMUnitNode, template="Contains::Execute(in Name,in Search,out Result)"}
   n25 = &RigVMFunction_ControlFlowBranch(Condition=@Contains.Result, ExecuteContext=@For_Each_2.ExecuteContext)
   n26 = RigVMDispatchNode(
    B=&RigUnit_GetTransform(bInitial=False, Item=@For_Each.Element).Transform
    A=&RigUnit_GetTransform(bInitial=False, Item=@For_Each_2.Element).Transform
   ) {template="DISPATCH_RigVMDispatch_CoreEquals(in A,in B,out Result)"}
   n27 = &RigVMFunction_ControlFlowBranch(Condition=@n26.Result, ExecuteContext=@n25.False)
   RigVMVariableNode(
    Value=RigVMDispatchNode(False=1, True=0.65, Condition=@n26.Result) {template="DISPATCH_RigVMDispatch_If(in Condition,in True,in False,out Result)"}.Result
    ExecuteContext=@n27.Completed
   ) {variable=Scale Factor}
   RigVMVariableNode(Value=&RigVMFunction_MathDoubleAdd(B=1, A=RigVMVariableNode() {variable=Control Stack}.Value).Result, ExecuteContext=@n27.True) {variable=Control Stack}
  ]}
]
hierarchy [MaxHandle {type=Curve, loc=<0 0 0>, rot=<0 0 0 1>}]
variables [
 All Bones {type="TArray<FRigElementKey>"}
 All Children Bones {type="TArray<FRigElementKey>"}
 Attach Controls {type="TArray<FRigElementKey>"}
 Attach Vis Channels {type="TArray<FRigElementKey>"}
 Children Controls {type="TArray<FRigElementKey>"}
 Control Stack {type=double, default=0}
 Driven Children Bones {type="TArray<FRigElementKey>"}
 Global Control {type=FRigElementKey, default=(Type=Bone,Name="")}
 Hand Controls {type="TArray<FRigElementKey>"}
 Hand Vis Channels {type="TArray<FRigElementKey>"}
 Local Control {type=FRigElementKey, default=(Type=Bone,Name="")}
 Offset {type=FRigElementKey, default=(Type=None,Name="")}
 Root {type=FRigElementKey, default=(Type=None,Name="")}
 Scale Factor {type=double, default=0}
]
